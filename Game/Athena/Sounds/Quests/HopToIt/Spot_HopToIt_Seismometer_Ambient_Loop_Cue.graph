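# Looping ambient sound cue for the seismometer in the Athena HopToIt quest, played quietly under the quests sound class.
# A looping wave player feeds an Enveloper that repeats indefinitely, with a smooth cubic volume curve that dips slightly mid-cycle and returns to full for gentle swelling.
# A seismometer attenuation asset handles spatial falloff, and the cue primes on load.
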SoundCue /Game/Athena/Sounds/Quests/HopToIt/Spot_HopToIt_Seismometer_Ambient_Loop_Cue
root=Enveloper_2
nodes [
 {
  id=Enveloper_2
  children [WavePlayer_1]
  properties {
   bLoop=true
   bLoopIndefinitely=true
   loopEnd=20
   pitchCurve {editorCurveData {keys [{time=0}]}}
   volumeCurve {editorCurveData {keys [
      {arriveTangent=-0.01, interpMode=RCIM_Cubic, time=0, value=1}
      <at 10 0.8 RCIM_Cubic>
      {interpMode=RCIM_Cubic, leaveTangent=0.01, time=20, value=1}
     ]}}
  }
 }
 {id=WavePlayer_1, properties {bLooping=true, soundWaveAssetPtr=&/Spot_HopToIt_Seismometer_Ambient_Loop}}
]
properties {
 attenuationSettings=&/Spot_Seismometer_Ambient_Att
 bPrimeOnLoad=true
 maxDistance=2510
 soundClassObject=&Fort_Quests
 volumeMultiplier=0.2
 attenuationOverrides {falloffDistance=700, attenuationShapeExtents {x=100}}
}
